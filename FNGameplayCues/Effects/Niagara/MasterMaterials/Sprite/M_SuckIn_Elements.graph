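# Translucent material for Niagara particles, including meshes, that appears to pull elements toward a Zero Point position.
# An optional world-offset branch scales the direction from actor position to that point by the squared Position_Offset dynamic parameter.
# Two panning electric-noise samples form a dissolve term, shaped by dynamic Exponent and Dissovle parameters and a curve atlas.
# Static switches tint the sprite texture by particle color and add an orange noise glow before emissive adjustments.
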Material /FNGameplayCues/Effects/Niagara/MasterMaterials/Sprite/M_SuckIn_Elements
graph {
 properties {
  bUsedWithMeshParticles=true
  bUsedWithNiagaraMeshParticles=true
  blendMode=BLEND_Translucent
  pixelDepthOffsetMode=PDOM_Legacy
  usedShadingModels=Default Lit
 }
 nodes [
  Add(A=MaterialFunctionCall() {materialFunction=&DebugTimeSine}.Result) {constB=-1}
  n1 = VectorParameter() {parameterName=DissovleCtrl, defaultValue=<rgba 3 0.02 0 0.1>}
  n2 = Panner(Coordinate=Multiply(A=TextureCoordinate(), B=AppendVector(A=@n1.R, B=@n1.R)), Speed=AppendVector(A=@n1.B, B=@n1.B)) {speedY=0.1}
  n3 = Panner(Coordinate=Multiply(A=AppendVector(A=@n1.G, B=@n1.G), B=TextureCoordinate()), Speed=AppendVector(A=@n1.A, B=@n1.A)) {speedX=0.1}
  n4 = Multiply(
   A=OneMinus(None=TextureSample(UVs=@n2) {samplerType=SAMPLERTYPE_LinearColor, texture=/Game/Characters/Player/Male/Medium/Bodies/M_MED_SRDrift/Textures/T_ElectricNoise_Mask}.R)
   B=OneMinus(None=TextureSample(UVs=@n3) {samplerType=SAMPLERTYPE_LinearColor, texture=&T_ElectricNoise_Mask}.R)
  )
  n5 = DynamicParameter() {paramNames [Position_Offset, SpeedRandom, Exponent, Dissovle]}
  n6 = Power(Base=OneMinus(None=@n4), Exp=@n5.Exponent) {constExponent=5}
  n7 = ParticleColor()
  n8 = CurveAtlasRowParameter(CurveTime=Add(A=@n4, B=@n5.Dissovle)) {
   atlas=/Game/Athena/LinearColorCurves/TempAthenaCurveAtlas
   curve=/Game/Athena/Prototype/Blueprints/NightNight/FX/FinalImpact/Curve_HT_Tomato_RC
   defaultValue=11
  }
  Add(A=Power(Base=@n6, Exp=@n7.A), B=Add(A=@n8.R) {constB=0})
  n9 = TextureSampleParameter2D() {parameterName=Texture, texture=/Game/Effects/Fort_Effects/World/Destruction/Textures/T_Prop_Leaf}
  n10 = StaticSwitchParameter(True=Multiply(A=@n9.RGB, B=@n7.RGB), False=Multiply(A=@n9.RGB) {constB=0.1}) {parameterName=UseParticle}
  n11 = Multiply(A=Constant3Vector() {constant=<rgba 1 0.4612 0.145 0>}, B=@n6)
  MaterialFunctionCall(Emissive=StaticSwitchParameter(True=Add(A=@n10, B=@n11), False=@n10) {parameterName=UseTransition}) {materialFunction=&MF_FX_EmissiveAdjustments}
  Reroute(None=@n9.RGB)
  n12 = VectorParameter() {parameterName=ZeroPoint_positon, defaultValue=<rgba 66048 -43008 0 1>}
  n13 = Normalize(VectorInput=Subtract(A=ComponentMask(None=@n12.RGB) {b=true, g=true, r=true}, B=ActorPositionWS()))
  n14 = ScalarParameter() {defaultValue=100, parameterName=DistanceMultiplication}
  n15 = DynamicParameter() {paramNames [Position_Offset, SpeedRandom, Exponent, Dissovle]}
  StaticSwitchParameter(
   True=Multiply(A=Multiply(A=@n13, B=@n14) {constB=1000}, B=Multiply(A=@n15.Position_Offset, B=@n15.Position_Offset)) {constB=100}
   False=Constant()
  ) {defaultValue=true, parameterName=UseWOP}
 ]
}
